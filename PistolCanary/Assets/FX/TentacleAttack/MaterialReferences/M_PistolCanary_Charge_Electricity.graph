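# Translucent unlit Niagara ribbon material for a wavering electric beam, named as the charge electricity of PistolCanary's TentacleAttack FX.
# Panned noise shifts the center of sphere masks across the ribbon's V axis, forming a jagged core and faint glow faded at the edges.
# Dynamic parameters drive core opacity, width, offset, distortion and tiling, though YTile is unused.
# The core mask blends EdgeColor to CoreColor, with particle color tinting the result and scaling opacity.
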Material /PistolCanary/Assets/FX/TentacleAttack/MaterialReferences/M_PistolCanary_Charge_Electricity
graph {
 properties {
  bAutomaticallySetUsageInEditor=false
  bUsedWithNiagaraRibbons=true
  blendMode=BLEND_Translucent
  shadingModel=MSM_Unlit
  usedShadingModels=Unlit
 }
 nodes [
  n1 = DynamicParameter() {parameterIndex=1, paramNames [XTile, YTile, "", ""]}
  n2 = DynamicParameter() {defaultValue=<rgba 1 0.05 1 1>, paramNames [CoreOpacity, CoreWidth, BeamOffset, DistortionStrength]}
  n3 = Add(
   A=Multiply(A=TextureCoordinate() {uTiling=1.4, vTiling=0.1}, B=AppendVector(A=@n1.XTile, B=@n1.XTile))
   B=AppendVector(A=@n2.BeamOffset, B=Constant())
  )
  MaterialFunctionCall(UVs=@n3, "Rotation Angle (0-1)"=Constant()) {materialFunction=&CustomRotator}
  n4 = ParticleColor()
  n5 = ComponentMask(None=TextureCoordinate()) {g=true}
  n6 = Constant() {r=0.5}
  n7 = SphereMask(A=@n5, B=@n6) {attenuationRadius=0.5, hardnessPercent=0}
  n8 = TextureSample(UVs=Panner(Coordinate=@n3, Speed=VectorParameter() {defaultValue=<rgba 0.1 0 0 1>}.RGB)) {samplerType=SAMPLERTYPE_Grayscale, texture=/Game/Global/Textures/Default/Noise/TilingNoise06}
  n9 = Add(A=@n6, B=Multiply(A=ConstantBiasScale(None=@n8.G) {bias=-0.5, scale=2}, B=Multiply(A=@n2.DistortionStrength) {constB=0.2}) {constB=0.2})
  n10 = SphereMask(A=@n5, B=@n9) {attenuationRadius=0.5, hardnessPercent=0}
  n11 = Multiply(A=SphereMask(A=@n5, B=@n9, Radius=@n2.CoreWidth) {attenuationRadius=0.05, hardnessPercent=0}, B=@n2.CoreOpacity)
  n12 = Add(A=Multiply(A=Multiply(A=@n10, B=@n10)) {constB=0.1}, B=Multiply(A=@n11, B=@n11))
  Multiply(A=@n4.A, B=Saturate(None=Multiply(A=@n7, B=@n12)))
  Multiply(
   A=LinearInterpolate(
    A=VectorParameter() {parameterName=EdgeColor, defaultValue=<rgba 0.2964 0.0007 1 1>}.RGB
    B=VectorParameter() {parameterName=CoreColor, defaultValue=<rgba 1 0.1137 0.7825 1>}.RGB
    Alpha=@n11
   )
   B=@n4.RGB
  )
 ]
}
