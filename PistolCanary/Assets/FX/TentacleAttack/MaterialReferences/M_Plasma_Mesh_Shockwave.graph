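# Unlit translucent mesh-particle material for a plasma shockwave in the tentacle attack FX, previewed on a torus.
# Three panning smoke samples, optionally mip-blurred by static switch, combine with V gradients into particle-tinted glow and opacity, with pan speed and tiling from dynamic parameters.
# Opacity fades by view angle, camera distance and scene depth, skipping depth fade on low quality.
# Offset math pinches vertices toward particle position and adds noise-driven height.
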Material /PistolCanary/Assets/FX/TentacleAttack/MaterialReferences/M_Plasma_Mesh_Shockwave
graph {
 properties {
  bUsedWithMeshParticles=true
  bUsedWithNiagaraMeshParticles=true
  blendMode=BLEND_Translucent
  pixelDepthOffsetMode=PDOM_Legacy
  previewMesh=/Game/Effects/Fort_Effects/Mesh/Shockwave/SM_BigPortal_Torus_Cine
  shadingModel=MSM_Unlit
  usedShadingModels=Unlit
 }
 nodes [
  ActorPositionWS()
  n1 = Subtract(A=ParticlePositionWS(), B=WorldPosition().XYZ)
  n2 = Power(Base=SphereMask(A=TextureCoordinate() {coordinateIndex=1}, B=Constant() {r=0.5}, Radius=Constant() {r=0.5}, Hardness=Constant()))
  n3 = ScalarParameter() {defaultValue=5, parameterName=HeightNosieU}
  n4 = ScalarParameter() {defaultValue=0.1, parameterName=HeightNosieV}
  n5 = TextureSample(UVs=Multiply(A=TextureCoordinate(), B=AppendVector(A=@n3, B=@n4))) {samplerType=SAMPLERTYPE_Grayscale, texture=&LowResBlurredNoise}
  n6 = Constant3Vector() {constant=<rgba 0 0 1 0>}
  Add(
   A=Multiply(A=Multiply(A=@n1, B=@n2), B=DynamicParameter() {paramNames [WPOPinch,PanSpeed,Tiling,Param4]}.WPOPinch)
   B=Multiply(A=Multiply(A=@n5.B, B=@n6), B=ScalarParameter() {parameterName=HeightAmount})
  )
  n7 = ParticleColor()
  MaterialFunctionCall(Emissive=@n7.RGB) {materialFunction=&MF_FX_EmissiveAdjustments}
  n8 = DynamicParameter() {paramNames [WPOPinch,PanSpeed,Tiling,Param4]}
  n9 = Multiply(A=TextureCoordinate() {uTiling=21, vTiling=0.75}, B=@n8.Tiling)
  n10 = Multiply(A=@n8.PanSpeed, B=Time())
  n11 = ScalarParameter() {defaultValue=5, parameterName=MipLevelBlur}
  n12 = Multiply(A=TextureCoordinate() {uTiling=10, vTiling=0.4}, B=@n8.Tiling)
  n13 = Add(
   A=TextureSample!(UVs=Panner(Coordinate=@n9, Time=@n10) {speedX=0.2, speedY=-1}, Tex, Level=@n11, Apply View MipBias) {mipValueMode=TMVM_MipLevel, texture=&T_Inky_Smoke_Tile}.G
   B=TextureSample(UVs=Panner(Coordinate=@n12, Time=@n10) {speedX=-0.1, speedY=-1}, Level=@n11) {mipValueMode=TMVM_MipLevel, texture=&T_Inky_Smoke_Tile}.G
  )
  n14 = TextureCoordinate() {uTiling=4, vTiling=0.3}
  n15 = TextureSample(UVs=Panner(Coordinate=Multiply(A=@n14, B=@n8.Tiling), Time=@n10) {speedX=0.25, speedY=-1}, Level=@n11) {mipValueMode=TMVM_MipLevel, texture=&T_Inky_Smoke_Tile}
  n16 = Multiply(A=TextureCoordinate() {uTiling=21, vTiling=0.75}, B=@n8.Tiling)
  n17 = Multiply(A=TextureCoordinate() {uTiling=10, vTiling=0.4}, B=@n8.Tiling)
  n18 = Add(
   A=TextureSample!(UVs=Panner(Coordinate=@n16, Time=@n10) {speedX=0.2, speedY=-1}, Tex, Apply View MipBias) {texture=&T_Inky_Smoke_Tile}.G
   B=TextureSample(UVs=Panner(Coordinate=@n17, Time=@n10) {speedX=-0.1, speedY=-1}) {texture=&T_Inky_Smoke_Tile}.G
  )
  n19 = TextureCoordinate() {uTiling=4, vTiling=0.3}
  n20 = TextureSample(UVs=Panner(Coordinate=Multiply(A=@n19, B=@n8.Tiling), Time=@n10) {speedX=0.25, speedY=-1}) {texture=&T_Inky_Smoke_Tile}
  n21 = ComponentMask(None=TextureCoordinate()) {g=true}
  n22 = OneMinus(None=@n21)
  n23 = Multiply(A=@n21) {constB=3}
  n24 = Multiply(
   A=StaticSwitchParameter(True=Multiply(A=@n13, B=@n15.G), False=Multiply(A=@n18, B=@n20.G)) {parameterName=BlurTextures}
   B=Multiply(A=@n22, B=Clamp(None=@n23))
  )
  n25 = Power(Base=@n24) {constExponent=1.25}
  Multiply(A=@n25, B=@n7.RGB) {constB=10}
  n26 = Constant() {r=0.5}
  n27 = StaticSwitchParameter(True=@n15.G, False=@n20.G) {parameterName=BlurTextures}
  n28 = ComponentMask(None=Add(A=Multiply(A=@n27), B=TextureCoordinate())) {g=true}
  n29 = Clamp(None=Power(Base=OneMinus(None=@n28)) {constExponent=15})
  n30 = Multiply(A=Add(A=@n21) {constB=-0.25}) {constB=3}
  n31 = Multiply(A=Add(A=@n22) {constB=0.2}) {constB=3}
  n32 = Multiply(
   A=Saturate(None=LinearInterpolate(A=@n24, B=@n26, Alpha=@n29))
   B=Multiply(A=Power(Base=@n30) {constExponent=5}, B=Power(Base=@n31) {constExponent=5})
  )
  n33 = MaterialFunctionCall("FadeSensitivity (optional)"=ScalarParameter() {defaultValue=1.5, parameterName=FadeAngleSensitivity}) {materialFunction=&FadeBasedOnViewAngle}
  n34 = Multiply(
   A=Multiply(A=Multiply(A=@n32, B=@n33.Result), B=@n7.A) {constB=2}
   B=MaterialFunctionCall(Distance=ScalarParameter() {defaultValue=500, parameterName=DistanceFade}) {materialFunction=&Near_Camera_Fade}.Result
  )
  QualitySwitch(Default=DepthFade(Opacity=@n34, FadeDistance=ScalarParameter() {defaultValue=150, parameterName=DepthFadeDist}), Low=@n34)
  n35 = MaterialFunctionCall() {materialFunction=/Game/Athena/WinterArt/Materials/MF_SharedWinter}
  n36 = MaterialFunctionCall!(Use Alternate UVs, Alternate UVs) {materialFunction=/Game/Athena/Effects/Materials/Geode/Functions/MF_Geode_Crack_Colors}
  StaticSwitchParameter(
   True=LinearInterpolate(A=@n35.IceColor, B=@n35.IceColor, Alpha=@n25)
   False=LinearInterpolate(A=@n36.Cloud Color, B=@n36.Cloud Color, Alpha=@n25)
  ) {parameterName=Use Winter Colors}
  Transform(None=WorldPosition().XYZ) {transformSourceType=TRANSFORMSOURCE_World, transformType=TRANSFORM_ParticleWorld}
 ]
}
