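# Additive unlit material for a large lightning bolt particle mesh, animated through particle dynamic parameters.
# Vertex offset combines camera-distance-scaled normal push for width, stepped panning-noise wobble, sine wave and vertical arc, with vertex color masking width and wobble.
# A Travel-driven mask along UV V, eroded by noise as particle alpha drops, shapes visibility.
# Emissive blends tinted gradients with time-of-day HDR, a mobile fallback, optional eye adaptation and switchable meteor glitch.
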Material /PistolCanary/Assets/FX/TentacleAttack/MaterialReferences/M_SuperSport_Large_LightningBolt_Mesh
graph {
 properties {
  bUsedWithMeshParticles=true
  bUsedWithNanite=true
  bUsedWithNiagaraMeshParticles=true
  bUsedWithNiagaraSprites=true
  bUsedWithParticleSprites=true
  blendMode=BLEND_Additive
  pixelDepthOffsetMode=PDOM_Legacy
  shadingModel=MSM_Unlit
  usedShadingModels=Unlit
 }
 nodes [
  n1 = VertexNormalWS()
  n2 = DynamicParameter() {defaultValue=<rgba 0.5 1 1 1>, paramNames [Travel,WidthMult,WobbleMult,Random]}
  n3 = ScalarParameter() {defaultValue=1, parameterName=DistanceWidthMult}
  n4 = MaterialFunctionCall(Distance=ScalarParameter() {defaultValue=25000, parameterName=DistanceWidthMultRange}) {materialFunction=&Near_Camera_Fade}
  n5 = Multiply(
   A=ScalarParameter() {defaultValue=50, parameterName=Push}
   B=Multiply(A=@n2.WidthMult, B=LinearInterpolate(B=@n3, Alpha=@n4.Result) {constA=1})
  )
  n6 = VertexColor()
  n7 = LinearInterpolate(Alpha=@n6.R) {constA=0.5}
  n8 = Multiply(
   A=Multiply(A=Normalize(VectorInput=@n1), B=Multiply(A=@n5, B=@n7))
   B=MaterialFunctionCall(Distance=ScalarParameter() {defaultValue=25000, parameterName=NearCameraShrink}) {materialFunction=&Near_Camera_Fade}.Result
  )
  n9 = MaterialFunctionCall!(MaskScaleMult=ScalarParameter() {defaultValue=3, parameterName=GlitchMaskScaleMult}) {materialFunction=/Game/Athena/Effects/Materials/Meteor/MF_Glitch_FX_World_Mask}
  n10 = Add(A=@n2.WobbleMult, B=StaticSwitchParameter(True=Multiply(A=@n9.WorldSpaceMask) {constB=0.4}, False=Constant()) {parameterName=GlitchWPO})
  n11 = ScalarParameter() {defaultValue=500, parameterName=WobbleStrength}
  n12 = MaterialFunctionCall(Float2=TextureCoordinate() {uTiling=0}) {materialFunction=&BreakOutFloat2Components}
  n13 = ParticleRandom()
  n14 = Multiply(
   A=MaterialFunctionCall(X=@n12.R, Y=Add(A=@n12.G, B=@n13)) {materialFunction=&MakeFloat2}.Result
   B=ScalarParameter() {defaultValue=6, parameterName=WobbleTile}
  )
  n15 = Multiply(A=Time(), B=ScalarParameter() {defaultValue=0.25, parameterName=WobbleSpeed})
  n16 = ScalarParameter() {defaultValue=6, parameterName=WobbleStepSize}
  n17 = Multiply(
   A=TextureSample(UVs=Panner(Coordinate=@n14, Time=@n15) {speedX=0.2, speedY=2}) {samplerType=SAMPLERTYPE_LinearColor, texture=&HighContrastNoise}.RGB
   B=@n16
  ) {constB=10}
  n18 = Subtract(A=Divide(A=Round(None=@n17), B=@n16) {constB=10}) {constB=0.5}
  n19 = Multiply(A=@n6.G, B=Multiply(A=Multiply(A=@n10, B=@n11), B=Multiply(A=@n18) {constB=2}))
  n20 = Add(A=Time(), B=DynamicParameter() {defaultValue=<rgba 0.5 1 1 1>, paramNames [Travel,WidthMult,WobbleMult,Random]}.Random)
  n21 = VectorParameter() {parameterName=MeteorDirection, defaultValue=<rgba 0.7755 -0.4565 -0.4361 0>}
  n22 = ScalarParameter() {defaultValue=200, parameterName=GlitchWPODist}
  n23 = Multiply(A=Sine(None=@n20) {period=0.11}, B=Multiply(A=Multiply(A=@n21.RGB, B=@n22)))
  n24 = Multiply(A=LinearInterpolate(Alpha=@n9.DistanceFadeMask) {constB=3}, B=Multiply(A=Round(None=@n23), B=@n9.WorldSpaceMask))
  n25 = Constant()
  n26 = VectorParameter() {parameterName=WPOWaveDirection, defaultValue=<rgba 0 1 0.5 0>}
  n27 = TextureCoordinate()
  n28 = DynamicParameter() {parameterIndex=1, defaultValue=<rgba 1 0 1 0>, paramNames [WPOWaveTime, WPOWaveMult, WPOWaveLength, VerticalArc]}
  n29 = Add(A=Divide(A=ComponentMask(None=@n27) {g=true}, B=@n28.WPOWaveLength), B=@n28.WPOWaveTime)
  n30 = MaterialFunctionCall(Y=Sine(None=@n29), Z=Sine(None=Add(A=@n29) {constB=-0.75}) {period=0.75}) {materialFunction=&MakeFloat3}
  n31 = Multiply(A=Transform(None=Multiply(A=@n26.RGB, B=@n30.Result)) {transformSourceType=TRANSFORMSOURCE_Local}, B=@n28.WPOWaveMult)
  n32 = TextureCoordinate()
  n33 = Constant() {r=0.5}
  n34 = Power(Base=SphereMask(A=ComponentMask(None=@n32) {g=true}, B=@n33, Radius=@n33) {hardnessPercent=0}) {constExponent=0.5}
  n35 = TextureCoordinate()
  n36 = SphereMask(
   A=Power(Base=ComponentMask(None=@n35) {g=true}, Exp=ScalarParameter() {defaultValue=0.5, parameterName=VerticalArcPower})
   B=@n33
   Radius=@n33
  ) {hardnessPercent=0}
  n37 = Multiply(A=@n28.VerticalArc, B=Constant3Vector() {constant=<rgba 0 0 1 0>})
  Add(
   A=Add(A=Add(A=@n8, B=@n19), B=StaticSwitchParameter(True=@n24, False=@n25) {parameterName=GlitchWPO})
   B=Add(A=Multiply(A=@n31, B=@n34), B=Multiply(A=@n36, B=@n37))
  )
  n38 = TextureCoordinate() {uTiling=0, vTiling=12}
  n39 = ParticleColor()
  n40 = Multiply(A=TextureSample(UVs=Panner(Coordinate=@n38, Time=@n39.A) {speedY=-4}) {samplerType=SAMPLERTYPE_LinearColor, texture=&HighContrastNoise}.R) {constB=-1}
  n41 = OneMinus(None=@n39.A)
  n42 = ComponentMask(None=TextureCoordinate()) {g=true}
  n43 = LinearInterpolate(Alpha=@n2.Travel) {constA=-1, constB=2}
  Clamp(None=Add(A=Multiply(A=@n40, B=@n41), B=SphereMask(A=@n42, B=@n43, Hardness=@n2.Travel) {attenuationRadius=1, hardnessPercent=50}))
  MaterialFunctionCall() {materialFunction=&DebugTimeSine}
  n44 = Constant() {r=0.64}
  n45 = ComponentMask(None=TextureCoordinate()) {g=true}
  n46 = LinearInterpolate(
   B=VectorParameter() {parameterName=TopColor, defaultValue=<rgba 1 0.8136 0 0>}.RGB
   Alpha=Multiply(A=Multiply(A=@n45, B=@n45), B=ScalarParameter() {parameterName=UseTopColor})
  ) {constA=1}
  n47 = LinearInterpolate(
   A=VectorParameter() {parameterName=SecondaryColor, defaultValue=<rgba 0.3247 0.0365 1 1>}.RGB
   B=VectorParameter() {parameterName=PrimaryColor, defaultValue=<rgba 0.8803 0.0573 1 1>}.RGB
   Alpha=@n6.R
  )
  n48 = VectorParameter() {parameterName=Glitch2, defaultValue=<rgba 1 0 0.2301 0>}
  n49 = Multiply(A=ConstantBiasScale(None=Sine(None=@n20) {period=0.178}), B=Multiply(A=Multiply(A=@n9.WorldSpaceMask, B=@n9.WorldSpaceMask)) {constB=2})
  n50 = Add(
   A=Multiply(A=@n39.RGB, B=Multiply(A=@n46, B=@n47))
   B=StaticSwitchParameter(True=Multiply(A=@n48.RGB, B=@n49), False=Constant()) {parameterName=GlitchWPO}
  )
  n51 = Multiply(A=@n50, B=ScalarParameter() {defaultValue=40, parameterName=HDRMult})
  n52 = Multiply(
   A=MaterialFunctionCall(
    Color_Morning=@n44
    Color_Day=Constant() {r=1}
    Color_Evening=@n44
    Color_Night=VectorParameter() {parameterName=NightColorMult, defaultValue=<rgba 0.5 0.0948 0.2162 0>}.RGB
   ) {materialFunction=&Color_Time_of_Day}.Result
   B=MaterialFunctionCall(LDR Input=Clamp(None=@n50), HDR Tint=@n51, Desaturate Luminance=ScalarParameter() {defaultValue=0.7, parameterName=HDRDesaturate}) {materialFunction=&DeriveHDRfromLDR}.RemappedRange
  )
  n53 = ScalarParameter() {
   defaultValue=1
   desc=Default HDRMult is too bright for mobile. This allows to reduce for mobile when HDR is high.
   parameterName=MobileBrightnessMultiplier
  }
  n54 = Clamp(None=Multiply(A=@n51, B=@n53)) {maxDefault=20}
  n55 = FeatureLevelSwitch(Default=@n52, ES2=@n54, ES3_1=@n54, SM4_REMOVED=@n52)
  n56 = LinearInterpolate(A=EyeAdaptation().EyeAdaptation, Alpha=Saturate(None=ScalarParameter() {parameterName=Eye Adaption Weight}))
  MaterialFunctionCall(Emissive=LinearInterpolate(A=@n55, B=Divide(A=@n55, B=@n56), Alpha=ScalarParameter() {parameterName=Use Eye Adaption Exposure Control for Emissives})) {materialFunction=&MF_FX_EmissiveAdjustments}
 ]
}
