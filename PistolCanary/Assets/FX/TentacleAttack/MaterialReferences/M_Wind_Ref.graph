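# Translucent unlit Niagara mesh-particle material for tentacle-attack wind effects, drawing an animated, distorted glowing band.
# Subtracting bottom from top edge textures forms the tinted band, warped by panning macro and optional micro noise, with optional flame licks on top.
# Two counter-panning noise samples vary emissive; the band also drives opacity via optional fresnel, depth fade and particle alpha.
# Optional world position offset combines view-dependent SplineThicken and panning normal-map noise.
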Material /PistolCanary/Assets/FX/TentacleAttack/MaterialReferences/M_Wind_Ref
graph {
 properties {
  bUsedWithNiagaraMeshParticles=true
  blendMode=BLEND_Translucent
  pixelDepthOffsetMode=PDOM_Legacy
  shadingModel=MSM_Unlit
  usedShadingModels=Unlit
 }
 nodes [
  n1 = ScalarParameter() {desc=Switches between using UV channel 0 or 1 for mapping distortion, group=1. Base, parameterName=Use_UV1_ForDistortion, sortPriority=-1}
  n2 = Multiply(
   A=LinearInterpolate(A=TextureCoordinate(), B=TextureCoordinate() {coordinateIndex=1}, Alpha=@n1)
   B=AppendVector(
    A=ScalarParameter() {defaultValue=15, group=Edge Band Emissive Noise, parameterName=EdgeNoise_Tile_X}
    B=ScalarParameter() {defaultValue=1.5, group=Edge Band Emissive Noise, parameterName=EdgeNoise_Tile_Y}
   )
  )
  n3 = ScalarParameter() {defaultValue=0.2, group=Edge Band Emissive Noise, parameterName=EdgeNoise_A_Panner_X}
  n4 = ScalarParameter() {defaultValue=0.5, group=Edge Band Emissive Noise, parameterName=EdgeNoise_A_Panner_Y}
  n5 = TextureSampleParameter2D(UVs=Panner(Coordinate=@n2, Speed=AppendVector(A=@n3, B=@n4))) {
   desc=Adds noise to the emissive channel on the edge band
   group=Edge Band Emissive Noise
   parameterName=EdgeNoise_Texture
   texture=/Game/Characters/Player/Male/Medium/Bodies/M_MED_Viper/Textures/Glitch/T_SoftNoise
  }
  n6 = ScalarParameter() {defaultValue=-0.2, group=Edge Band Emissive Noise, parameterName=EdgeNoise_B_Panner_X}
  n7 = ScalarParameter() {defaultValue=0.2, group=Edge Band Emissive Noise, parameterName=EdgeNoise_B_Panner_Y}
  n8 = TextureSampleParameter2D(UVs=Panner(Coordinate=@n2, Speed=AppendVector(A=@n6, B=@n7))) {
   desc=Adds noise to the emissive channel on the edge band
   group=Edge Band Emissive Noise
   parameterName=EdgeNoise_Texture
   texture=&T_SoftNoise
  }
  n9 = Subtract(A=@n5.R, B=@n8.R)
  n10 = TextureCoordinate()
  n11 = TextureCoordinate() {coordinateIndex=1}
  n12 = ScalarParameter() {desc=Switches between using UV channel 0 or 1 for mapping distortion, group=1. Base, parameterName=Use_UV1_ForDistortion, sortPriority=-1}
  n13 = MaterialFunctionCall(Float2=Multiply(
    A=LinearInterpolate(A=@n10, B=@n11, Alpha=@n12)
    B=ScalarParameter() {defaultValue=1, group="2. Edge Distortion Band (Fresnel Distortion = 0)", parameterName=EdgeDistortion_Tiling}
   )) {materialFunction=&BreakOutFloat2Components}
  n14 = Add(
   A=@n13.G
   B=ScalarParameter() {defaultValue=0.3, group="2. Edge Distortion Band (Fresnel Distortion = 0)", parameterName=EdgeDistort_OffsetBottom}
  )
  n15 = ScalarParameter() {desc=Switches between using UV channel 0 or 1 for mapping distortion, group=1. Base, parameterName=Use_UV1_ForDistortion, sortPriority=-1}
  n16 = MaterialFunctionCall(Float2=LinearInterpolate(A=TextureCoordinate(), B=TextureCoordinate() {coordinateIndex=1}, Alpha=@n15)) {materialFunction=&BreakOutFloat2Components}
  n17 = Multiply(A=@n16.R, B=ScalarParameter() {defaultValue=1, group=Distortion_Macro, parameterName=Distortion_Macro_Tile_X, sortPriority=3})
  n18 = Multiply(A=@n16.G, B=ScalarParameter() {defaultValue=0.5, group=Distortion_Macro, parameterName=Distortion_Macro_Tile_Y, sortPriority=4})
  n19 = ScalarParameter() {defaultValue=0.1, group=Distortion_Macro, parameterName=Distortion_Macro_Panspeed_X, sortPriority=1}
  n20 = ScalarParameter() {defaultValue=0.25, group=Distortion_Macro, parameterName=Distortion_Macro_Panspeed_Y, sortPriority=2}
  n21 = TextureSampleParameter2D(UVs=Panner(Coordinate=AppendVector(A=@n17, B=@n18), Speed=AppendVector(A=@n19, B=@n20)) {speedY=0.2}) {
   group=Distortion_Macro
   parameterName=Macro_Distortion_Texture
   samplerSource=SSM_Wrap_WorldGroupSettings
   texture=&AtmosphericCloudNoise01
  }
  n22 = ScalarParameter() {defaultValue=1, group=Distortion_Macro, parameterName=Macro_Distortion_Mult}
  n23 = Clamp(None=Power(Base=Multiply(A=@n21.R, B=@n22), Exp=ScalarParameter() {defaultValue=1, group=Distortion_Macro, parameterName=Macro_Distortion_Exponent}))
  n24 = Multiply(A=AppendVector(A=@n23, B=Constant())) {constB=2}
  n25 = Multiply(A=AppendVector(A=Constant(), B=@n23)) {constB=2}
  n26 = Add(
   A=LinearInterpolate(
    A=Add(A=@n24) {constB=-1}
    B=Add(A=@n25) {constB=-1}
    Alpha=ScalarParameter() {defaultValue=1, group=Distortion_Macro, parameterName=MultAdd_XorY, sortPriority=7}
   )
   B=AppendVector(
    A=ScalarParameter() {group=Distortion_Macro, parameterName=Macro_Distortion_Offset_X, sortPriority=6}
    B=ScalarParameter() {defaultValue=1, group=Distortion_Macro, parameterName=Macro_Distortion_Offset_Y, sortPriority=7}
   )
  ) {constB=0}
  n27 = ScalarParameter() {defaultValue=0.1, group=Distortion_Macro, parameterName=Macro_Distortion_Intensity, sortPriority=5}
  n28 = ScalarParameter() {
   defaultValue=1
   desc="Large Brush stroke distortion. Good for overall silhouette changes. "
   group=Distortion_Macro
   parameterName=Use_Macro_Distortion
   sortPriority=-1
  }
  n29 = ScalarParameter() {desc=Switches between using UV channel 0 or 1 for mapping distortion, group=1. Base, parameterName=Use_UV1_ForDistortion, sortPriority=-1}
  n30 = MaterialFunctionCall(Float2=LinearInterpolate(A=TextureCoordinate(), B=TextureCoordinate() {coordinateIndex=1}, Alpha=@n29)) {materialFunction=&BreakOutFloat2Components}
  n31 = Multiply(A=@n30.R, B=ScalarParameter() {defaultValue=3, group=Distortion_Micro, parameterName=Distortion_Micro_Tile_X, sortPriority=3})
  n32 = Multiply(A=@n30.G, B=ScalarParameter() {defaultValue=0.5, group=Distortion_Micro, parameterName=Distortion_Micro_Tile_Y, sortPriority=4})
  n33 = ScalarParameter() {defaultValue=0.1, group=Distortion_Micro, parameterName=Distortion_Micro_Panspeed_X, sortPriority=1}
  n34 = ScalarParameter() {defaultValue=0.5, group=Distortion_Micro, parameterName=Distortion_Micro_Panspeed_Y, sortPriority=2}
  n35 = TextureSampleParameter2D(UVs=Panner(Coordinate=AppendVector(A=@n31, B=@n32), Speed=AppendVector(A=@n33, B=@n34)) {speedY=0.2}) {
   group=Distortion_Micro
   parameterName=Micro_Distortion_Texture
   samplerSource=SSM_Wrap_WorldGroupSettings
   texture=/Game/Global/Textures/Default/Noise/T_Dudebro_Noise_01
  }
  n36 = Add(A=Multiply(A=AppendVector(A=@n35.G, B=@n35.G)) {constB=2}) {constB=-1}
  n37 = ScalarParameter() {defaultValue=0.0125, group=Distortion_Micro, parameterName=Micro_Distortion_Intensity, sortPriority=5}
  n38 = ScalarParameter() {
   desc="Small detail distortion. Good for giving extra detail to compliment the silhouette. "
   group=Distortion_Micro
   parameterName=Use_Micro_Distortion
   sortPriority=-1
  }
  n39 = Add(A=LinearInterpolate(B=Multiply(A=@n26, B=@n27) {constB=0.05}, Alpha=@n28), B=LinearInterpolate(B=Multiply(A=@n36, B=@n37), Alpha=@n38))
  n40 = TextureSampleParameter2D(UVs=Add(A=MaterialFunctionCall(R=@n13.R, G=@n14) {materialFunction=&AppendMany}.RG, B=@n39)) {
   desc=Used to control the bottom edge position of the band
   group="2. Edge Distortion Band (Fresnel Distortion = 0)"
   parameterName=EdgeDistortion_Bottom_Texture
   samplerSource=SSM_Clamp_WorldGroupSettings
   texture=/Game/Weapons/FORT_Melee/Pickaxe_Obsidian_Female/Effects/T_FX_Obsidian_FlameDistortion
  }
  n41 = TextureCoordinate()
  n42 = TextureCoordinate() {coordinateIndex=1}
  n43 = ScalarParameter() {desc=Switches between using UV channel 0 or 1 for mapping distortion, group=1. Base, parameterName=Use_UV1_ForDistortion, sortPriority=-1}
  n44 = MaterialFunctionCall(Float2=Multiply(
    A=LinearInterpolate(A=@n41, B=@n42, Alpha=@n43)
    B=ScalarParameter() {defaultValue=1, group="2. Edge Distortion Band (Fresnel Distortion = 0)", parameterName=EdgeDistortion_Tiling}
   )) {materialFunction=&BreakOutFloat2Components}
  n45 = ScalarParameter() {desc=Switches between using UV channel 0 or 1 for mapping distortion, group=1. Base, parameterName=Use_UV1_ForDistortion, sortPriority=-1}
  n46 = LinearInterpolate(A=TextureCoordinate(), B=TextureCoordinate() {coordinateIndex=1}, Alpha=@n45)
  n47 = AppendVector(
   A=ScalarParameter() {defaultValue=1, group="2. Edge Distortion Band (Fresnel Distortion = 0)", parameterName=TopEdge_LerpMask_Tiling_X}
   B=ScalarParameter() {group="2. Edge Distortion Band (Fresnel Distortion = 0)", parameterName=TopEdge_LerpMask_Tiling_Y}
  )
  n48 = ScalarParameter() {defaultValue=-0.25, group="2. Edge Distortion Band (Fresnel Distortion = 0)", parameterName=TopEdge_LerpMask_Speed_X}
  n49 = ScalarParameter() {defaultValue=0.1, group="2. Edge Distortion Band (Fresnel Distortion = 0)", parameterName=TopEdge_LerpMask_Speed_Y}
  n50 = TextureSampleParameter2D(UVs=Panner(Coordinate=Multiply(A=@n46, B=@n47), Speed=AppendVector(A=@n48, B=@n49))) {
   desc=Adds extra peaks and valleys by masking the top edge distortion edge height
   group="2. Edge Distortion Band (Fresnel Distortion = 0)"
   parameterName=EdgeDistortion_TopEdge_LerpMask
   texture=&T_SoftNoise
  }
  n51 = Add(
   A=@n44.G
   B=LinearInterpolate(
    A=ScalarParameter() {defaultValue=0.4, group="2. Edge Distortion Band (Fresnel Distortion = 0)", parameterName=EdgeDistort_OffsetTop_LerpMask_Min}
    B=ScalarParameter() {defaultValue=0.4, group="2. Edge Distortion Band (Fresnel Distortion = 0)", parameterName=EdgeDistort_OffsetTop_LerpMask_Max}
    Alpha=@n50.R
   )
  )
  n52 = ScalarParameter() {desc=Switches between using UV channel 0 or 1 for mapping distortion, group=1. Base, parameterName=Use_UV1_ForDistortion, sortPriority=-1}
  n53 = LinearInterpolate(A=TextureCoordinate(), B=TextureCoordinate() {coordinateIndex=1}, Alpha=@n52)
  n54 = AppendVector(
   A=ScalarParameter() {defaultValue=10, group=Distortion_FlameLicks, parameterName=Lick_Distortion_Tiling_X}
   B=ScalarParameter() {defaultValue=2, group=Distortion_FlameLicks, parameterName=Lick_Distortion_Tiling_Y}
  )
  n55 = ScalarParameter() {defaultValue=0.1, group=Distortion_FlameLicks, parameterName=Lick_Distortion_Speed_X}
  n56 = ScalarParameter() {defaultValue=0.1, group=Distortion_FlameLicks, parameterName=Lick_Distortion_Speed_Y}
  n57 = TextureSampleParameter2D(UVs=Panner(Coordinate=Multiply(A=@n53, B=@n54), Speed=AppendVector(A=@n55, B=@n56))) {
   group=Distortion_FlameLicks
   parameterName=Lick_Distortion_Texture
   samplerType=SAMPLERTYPE_Masks
   texture=/Game/Global/Textures/Default/Noise/T_CellularNoise
  }
  n58 = Clamp(None=OneMinus(None=@n57.R))
  n59 = ScalarParameter() {desc=Switches between using UV channel 0 or 1 for mapping distortion, group=1. Base, parameterName=Use_UV1_ForDistortion, sortPriority=-1}
  n60 = LinearInterpolate(A=TextureCoordinate(), B=TextureCoordinate() {coordinateIndex=1}, Alpha=@n59)
  n61 = AppendVector(
   A=ScalarParameter() {defaultValue=2, group=Distortion_FlameLicks, parameterName=Lick_Mask_Tiling_X, sliderMin=1}
   B=ScalarParameter() {group=Distortion_FlameLicks, parameterName=Lick_Mask_Tiling_Y}
  )
  n62 = ScalarParameter() {defaultValue=-0.1, group=Distortion_FlameLicks, parameterName=Lick_Mask_Speed_X}
  n63 = ScalarParameter() {defaultValue=0.02, group=Distortion_FlameLicks, parameterName=Lick_Mask_Speed_Y}
  n64 = TextureSampleParameter2D(UVs=Panner(Coordinate=Multiply(A=@n60, B=@n61), Speed=AppendVector(A=@n62, B=@n63))) {
   desc=Uses a texture to mask the distortion. Can be used to give gaps with peaks and valleys.
   group=Distortion_FlameLicks
   parameterName=Lick_Mask_Texture
   texture=/Game/Environments/Foliage/Textures/T_NoiseMask03_Sparce
  }
  n65 = ScalarParameter() {defaultValue=2, group=Distortion_FlameLicks, parameterName=Lick_Mask_Exponent}
  n66 = LinearInterpolate(
   A=ScalarParameter() {defaultValue=0.01, group=Distortion_FlameLicks, parameterName=Lick_Mask_Intensity_Minimum}
   B=ScalarParameter() {defaultValue=1, group=Distortion_FlameLicks, parameterName=Lick_Mask_Intensity_Maximum}
   Alpha=Clamp(None=Power(Base=@n64.R, Exp=@n65))
  )
  n67 = ScalarParameter() {
   desc=" Adds a distortion that resembles flame-licks. Thin fingers that reach out."
   group=Distortion_FlameLicks
   parameterName=Use_Distortion_Lick
   sortPriority=-1
  }
  n68 = LinearInterpolate(
   B=Multiply(A=Multiply(A=@n58, B=@n66), B=ScalarParameter() {defaultValue=0.1, group=Distortion_FlameLicks, parameterName=Lick_Distortion_Intensity}) {constB=0.1}
   Alpha=@n67
  )
  n69 = TextureSampleParameter2D(UVs=Add(A=AppendVector(A=@n44.R, B=@n51), B=Add(A=@n39, B=@n68))) {
   desc=Used to control the top edge position of the band
   group="2. Edge Distortion Band (Fresnel Distortion = 0)"
   parameterName=EdgeDistortion_Top_Texture
   samplerSource=SSM_Clamp_WorldGroupSettings
   texture=&T_FX_Obsidian_FlameDistortion
  }
  n70 = VectorParameter() {group="2. Edge Distortion Band (Fresnel Distortion = 0)", parameterName=EdgeDistort_Color, defaultValue=<rgba 0.9823 0.4911 0.9004 1>}
  n71 = Multiply(A=Subtract(A=@n69.RGB, B=@n40.RGB), B=@n70.RGB)
  n72 = Multiply(
   A=@n71
   B=LinearInterpolate(
    A=ScalarParameter() {defaultValue=0.8, group=Edge Band Emissive Noise, parameterName=EdgeNoise_Emissive_Min}
    B=ScalarParameter() {defaultValue=10, group=Edge Band Emissive Noise, parameterName=EdgeNoise_Emissive_Max}
    Alpha=Clamp(None=@n9)
   )
  )
  MaterialFunctionCall(Emissive=@n72) {materialFunction=&MF_FX_EmissiveAdjustments}
  n73 = DepthFade(Opacity=Constant() {r=1}, FadeDistance=ScalarParameter() {group=1. Base, parameterName=DepthFade_Distance, sortPriority=34})
  n74 = ScalarParameter() {defaultValue=1, group=Fresnel Opacity, parameterName=Opacity-Fresnel_Brightness}
  n75 = ScalarParameter() {defaultValue=20, group=Fresnel Opacity, parameterName=Opacity-Fresnel_Exp}
  n76 = Normalize(VectorInput=Subtract(A=CameraPositionWS(), B=ActorPositionWS()))
  n77 = Transform(None=VectorParameter() {group=SplineThicken, parameterName=Lookat-VectortoCompare, defaultValue=<rgba 0 0 1 1>}.RGB) {transformSourceType=TRANSFORMSOURCE_Local}
  n78 = Multiply(A=Abs(None=DotProduct(A=@n76, B=@n77)), B=ScalarParameter() {defaultValue=1, group=SplineThicken, parameterName=Lookat-Dot_Mult})
  n79 = ScalarParameter() {defaultValue=1, group=SplineThicken, parameterName=Lookat-Dot_Exp}
  n80 = LinearInterpolate(
   A=ScalarParameter() {defaultValue=6, group=SplineThicken, parameterName=Thickness-LookAway}
   B=ScalarParameter() {defaultValue=6, group=SplineThicken, parameterName=Thickness-LookAt}
   Alpha=Clamp(None=Power(Base=@n78, Exp=@n79))
  )
  n81 = MaterialFunctionCall(
   WidthBase=@n80
   WidthTip=@n80
   Expand U or V UV Channel=StaticBoolParameter() {group=SplineThicken, parameterName="Expand U or V UV Channel (B)"}
  ) {materialFunction=&SplineThicken}
  n82 = LinearInterpolate(A=VertexNormalWS(), B=@n81.Normal, Alpha=ScalarParameter() {group=SplineThicken, parameterName=Use_SplineThicken, sortPriority=-1})
  n83 = LinearInterpolate(
   A=@n71
   B=Multiply(
    A=@n71
    B=MaterialFunctionCall(Brightness=@n74, Exponent=@n75, Vertex / Pixel Normal=@n82) {materialFunction=&MF_Fresnel_FX}.Fresnel
   )
   Alpha=ScalarParameter() {group=Fresnel Opacity, parameterName=Use_Fresnel_Opacity_Mask, sortPriority=-1}
  )
  n84 = FeatureLevelSwitch(Default=@n73, ES2=@n83, ES3_1=@n83, SM4_REMOVED=@n73)
  Multiply(
   A=LinearInterpolate(A=@n83, B=Multiply(A=@n83, B=@n84), Alpha=ScalarParameter() {group=1. Base, parameterName=Use_DepthFade, sortPriority=33})
   B=ParticleColor().A
  )
  Reroute(None=@n72)
  n85 = ScalarParameter() {group=SplineThicken, parameterName=Use_SplineThicken, sortPriority=-1}
  n86 = LinearInterpolate(
   A=TextureCoordinate()
   B=TextureCoordinate() {coordinateIndex=1}
   Alpha=ScalarParameter() {defaultValue=1, group=WPO, parameterName=Use_UV0or1, sortPriority=-1}
  )
  n87 = AppendVector(
   A=ScalarParameter() {defaultValue=1, group=WPO, parameterName=WPO_Tiling_X}
   B=ScalarParameter() {defaultValue=1, group=WPO, parameterName=WPO_Tiling_Y}
  )
  n88 = ScalarParameter() {defaultValue=0.025, group=WPO, parameterName=WPO_Speed_X}
  n89 = ScalarParameter() {defaultValue=0.1, group=WPO, parameterName=WPO_Speed_Y}
  n90 = TextureSampleParameter2D(UVs=Panner(Coordinate=Multiply(A=@n86, B=@n87), Speed=AppendVector(A=@n88, B=@n89))) {group=WPO, parameterName=WPO Texture, samplerType=SAMPLERTYPE_Normal, texture=/Game/Global/Textures/Default/Noise/T_Normal_Noise01}
  n91 = Multiply(A=@n90.RGB, B=ScalarParameter() {defaultValue=2, group=WPO, parameterName=WPO_Intensity})
  n92 = ScalarParameter() {group=WPO, parameterName=Use_WPO, sortPriority=-1}
  Reroute(None=Add(A=LinearInterpolate(B=@n81.WorldPositionOffset, Alpha=@n85), B=LinearInterpolate(B=@n91, Alpha=@n92)))
 ]
}
